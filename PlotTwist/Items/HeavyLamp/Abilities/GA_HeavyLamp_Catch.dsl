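; Catch ability for the HeavyLamp item, a child ability that adjusts montage, weapon and camera behavior around its parent's catch logic.
; On activation, a pawn with the blocking tag swaps to the catch-to-block montage before the parent logic runs.
; On end, it clears InstantOn on a Lobster (lightsaber) weapon so later catches play the power-on animation, and pops the camera mode unless the pawn is blocking.
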(event Ability|EventActivateAbility
  (bind _outfortplayerpawn (CallFunction|GetAvatarasValidFortPlayerPawn)
    (:ValidFortPlayerPawn
      (if (GameplayTags|HasMatchingGameplayTag _outfortplayerpawn (Variables|Default|GetBlockingTag))
        (Variables|CatchVars|SetMontageToPlay (Variables|Default|GetCatchToBlockMontage)) ; if not blocking, will use standard catch\notherwise override with catch anim
        (else
          _))
      (|CallParentFunction))
    (:AvatarCastFailed
      _)))

(event Ability|EventOnEndAbility
  (params (bWasCancelled :type Boolean))
  (bind _playerpawn (Variables|CatchVars|GetPlayerPawn))
  (|CallParentFunction bWasCancelled)
  ;[1 1 1] sets the lightsaber back to non-instant turn on so that SUBSEQUENT catches will play the standard "power on" animation
  (bind _asb_athena_galileo_lobster_parent (Utilities|Casting|CastToB_Athena_Galileo_Lobster_Parent (Variables|/Script/FortniteGame.FortPawn|GetCurrentWeapon _playerpawn))
    (:then
      (Variables|/Galileo/Items/Lobster/Actors/B_Athena_Galileo_Lobster_Parent.B_Athena_Galileo_Lobster_Parent_C|SetInstantOn false _asb_athena_galileo_lobster_parent))
    (:CastFailed
      _))
  (if (not (GameplayTags|HasMatchingGameplayTag _playerpawn (Variables|Default|GetBlockingTag)))
    (Camera|PopCameraMode)))
